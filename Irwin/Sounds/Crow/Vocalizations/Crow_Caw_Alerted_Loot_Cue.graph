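# Sound Cue for a crow's alerted caw, named for loot alerts, played as an attenuated creature vocalization.
# A quality-level node sends its first two levels to a random pick of eight long caw waves and the last two to a three-wave pick, which is excluded from branch culling.
# Each branch varies volume through a modulator.
# The cue lowers pitch, boosts volume and overrides concurrency to limit simultaneous caws.
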SoundCue /Irwin/Sounds/Crow/Vocalizations/Crow_Caw_Alerted_Loot_Cue
root=QualityLevel_0
nodes [
 {id=Modulator_0, children [Random_0], properties {volumeMax=1, volumeMin=0.85}}
 {id=Modulator_3, children [Random_1], properties {volumeMax=1, volumeMin=0.75}}
 {id=QualityLevel_0, children [Modulator_0,Modulator_0,Modulator_3,Modulator_3]}
 {
  id=Random_0
  children [WavePlayer_15, WavePlayer_16, WavePlayer_10, WavePlayer_11, WavePlayer_19, WavePlayer_20, WavePlayer_21, WavePlayer_22]
  properties {weights [1,1,1,1,1,1,1,1]}
 }
 {id=Random_1, children [WavePlayer_23, WavePlayer_24, WavePlayer_25], properties {bShouldExcludeFromBranchCulling=true, weights [1,1,1]}}
 {id=WavePlayer_10, properties {soundWaveAssetPtr=&/Crow_Caw_Long_03}}
 {id=WavePlayer_11, properties {soundWaveAssetPtr=&/Crow_Caw_Long_04}}
 {id=WavePlayer_15, properties {soundWaveAssetPtr=&/Crow_Caw_Long_01}}
 {id=WavePlayer_16, properties {soundWaveAssetPtr=&/Crow_Caw_Long_02}}
 {id=WavePlayer_19, properties {soundWaveAssetPtr=&/Crow_Caw_Long_05}}
 {id=WavePlayer_20, properties {soundWaveAssetPtr=&/Crow_Caw_Long_06}}
 {id=WavePlayer_21, properties {soundWaveAssetPtr=&/Crow_Caw_Long_07}}
 {id=WavePlayer_22, properties {soundWaveAssetPtr=&/Crow_Caw_Long_08}}
 {id=WavePlayer_23, properties {soundWaveAssetPtr=&Crow_Caw_Long_01}}
 {id=WavePlayer_24, properties {soundWaveAssetPtr=&Crow_Caw_Long_02}}
 {id=WavePlayer_25, properties {soundWaveAssetPtr=&Crow_Caw_Long_03}}
]
properties {
 attenuationSettings=/Irwin/Sounds/Crow/Crow_Vocalization_Att
 bOverrideConcurrency=true
 maxDistance=3300
 pitchMultiplier=0.9
 soundClassObject=&Fort_Creatures_Vocalization_Tactical
 volumeMultiplier=2.45
 concurrencyOverrides {maxCount=3, volumeScale=0.7}
}
