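# Crow death vocalization cue (named for loot) that layers a pitched-up long caw over wing take-off foley.
# A quality-level switch selects one of two branches, one with larger random sample pools than the other.
# Each branch doubles its foley through a slightly delayed path from the same random node, with separate pitch modulation for caw and foley.
# Attenuation is shared crow vocalization; per-owner concurrency allows one instance and blocks new ones.
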SoundCue /Irwin/Sounds/Crow/Vocalizations/Crow_Caw_Death_Loot_Cue
root=Modulator_0
nodes [
 {id=Delay_0, children [Random_2], properties {delayMax=0.1, delayMin=0.05}}
 {id=Delay_1, children [Random_4], properties {delayMax=0.1, delayMin=0.05}}
 {id=Mixer_1, children [Modulator_3,Modulator_1], properties {inputVolume [1,1]}}
 {id=Mixer_2, children [Modulator_4,Modulator_2], properties {inputVolume [1,1]}}
 {id=Mixer_3, children [Random_2,Delay_0], properties {inputVolume [1,1]}}
 {id=Mixer_4, children [Random_4,Delay_1], properties {inputVolume [1,1]}}
 {id=Modulator_0, children [QualityLevel_0], properties {volumeMax=1, volumeMin=1}}
 {id=Modulator_1, children [Mixer_3], properties {pitchMax=1.15, pitchMin=1.05, volumeMax=1, volumeMin=1}}
 {id=Modulator_2, children [Mixer_4], properties {pitchMax=1.15, pitchMin=1.05, volumeMax=1, volumeMin=1}}
 {id=Modulator_3, children [Random_1], properties {pitchMax=1.25, pitchMin=1.2, volumeMax=1, volumeMin=1}}
 {id=Modulator_4, children [Random_5], properties {pitchMax=1.25, pitchMin=1.2, volumeMax=1, volumeMin=1}}
 {id=QualityLevel_0, children [Mixer_1,Mixer_1,Mixer_2,Mixer_2]}
 {
  id=Random_1
  children [WavePlayer_0, WavePlayer_1, WavePlayer_10, WavePlayer_3, WavePlayer_4, WavePlayer_5]
  properties {weights [1,1,1,1,1,1]}
 }
 {id=Random_2, children [WavePlayer_14, WavePlayer_15, WavePlayer_16], properties {weights [1,1,1]}}
 {id=Random_4, children [WavePlayer_17, WavePlayer_18], properties {bShouldExcludeFromBranchCulling=true, weights [1,1]}}
 {id=Random_5, children [WavePlayer_23, WavePlayer_24, WavePlayer_25], properties {bShouldExcludeFromBranchCulling=true, weights [1,1,1]}}
 {id=WavePlayer_0, properties {soundWaveAssetPtr=&/Crow_Caw_Long_01}}
 {id=WavePlayer_1, properties {soundWaveAssetPtr=&/Crow_Caw_Long_02}}
 {id=WavePlayer_10, properties {soundWaveAssetPtr=&/Crow_Caw_Long_03}}
 {id=WavePlayer_14, properties {soundWaveAssetPtr=/Irwin/Sounds/Crow/Foley/Crow_Foley_Fly_Start_02}}
 {id=WavePlayer_15, properties {soundWaveAssetPtr=/Irwin/Sounds/Crow/Foley/Crow_Foley_Fly_Start_03}}
 {id=WavePlayer_16, properties {soundWaveAssetPtr=/Irwin/Sounds/Crow/Foley/Crow_Foley_Fly_Start_04}}
 {id=WavePlayer_17, properties {soundWaveAssetPtr=/Irwin/Sounds/Crow/Foley/Crow_Foley_Fly_Start_01}}
 {id=WavePlayer_18, properties {soundWaveAssetPtr=&Crow_Foley_Fly_Start_02}}
 {id=WavePlayer_23, properties {soundWaveAssetPtr=&Crow_Caw_Long_01}}
 {id=WavePlayer_24, properties {soundWaveAssetPtr=&Crow_Caw_Long_02}}
 {id=WavePlayer_25, properties {soundWaveAssetPtr=&Crow_Caw_Long_03}}
 {id=WavePlayer_3, properties {soundWaveAssetPtr=&/Crow_Caw_Long_06}}
 {id=WavePlayer_4, properties {soundWaveAssetPtr=&/Crow_Caw_Long_07}}
 {id=WavePlayer_5, properties {soundWaveAssetPtr=&/Crow_Caw_Long_08}}
]
properties {
 attenuationSettings=/Irwin/Sounds/Crow/Crow_Vocalization_Att
 bOverrideConcurrency=true
 maxDistance=3300
 pitchMultiplier=0.9
 soundClassObject=&Fort_Creatures_Vocalization_Tactical
 volumeMultiplier=5
 concurrencyOverrides {bLimitToOwner=true, maxCount=1, resolutionRule=PreventNew}
}
