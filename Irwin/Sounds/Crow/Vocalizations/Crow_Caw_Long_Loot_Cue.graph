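# Sound cue for the crow's long caw "Loot" vocalization, played with crow attenuation and the tactical creature vocalization sound class.
# A quality-level node picks an eight-wave random pool for the first two levels and a three-wave pool for the last two.
# Both weighted pools include an empty branch, so over half of triggers are silent.
# A modulator varies volume, and concurrency allows one instance per owner, blocking new ones.
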SoundCue /Irwin/Sounds/Crow/Vocalizations/Crow_Caw_Long_Loot_Cue
root=Modulator_0
nodes [
 {id=Modulator_0, children [QualityLevel_0], properties {volumeMax=1, volumeMin=0.85}}
 {id=QualityLevel_0, children [Random_0,Random_0,Random_3,Random_3]}
 {
  id=Random_0
  children [WavePlayer_0, WavePlayer_1, WavePlayer_10, WavePlayer_11, WavePlayer_2, WavePlayer_3, WavePlayer_4, WavePlayer_5, null]
  properties {bRandomizeWithoutReplacement=false, weights [0.056,0.056,0.056,0.056,0.056,0.056,0.056,0.056,0.55]}
 }
 {
  id=Random_3
  children [WavePlayer_17, WavePlayer_18, WavePlayer_19, null]
  properties {bRandomizeWithoutReplacement=false, bShouldExcludeFromBranchCulling=true, weights [0.15,0.15,0.15,0.55]}
 }
 {id=WavePlayer_0, properties {soundWaveAssetPtr=&/Crow_Caw_Long_01}}
 {id=WavePlayer_1, properties {soundWaveAssetPtr=&/Crow_Caw_Long_02}}
 {id=WavePlayer_10, properties {soundWaveAssetPtr=&/Crow_Caw_Long_03}}
 {id=WavePlayer_11, properties {soundWaveAssetPtr=&/Crow_Caw_Long_04}}
 {id=WavePlayer_17, properties {soundWaveAssetPtr=&Crow_Caw_Long_01}}
 {id=WavePlayer_18, properties {soundWaveAssetPtr=&Crow_Caw_Long_02}}
 {id=WavePlayer_19, properties {soundWaveAssetPtr=&Crow_Caw_Long_03}}
 {id=WavePlayer_2, properties {soundWaveAssetPtr=&/Crow_Caw_Long_05}}
 {id=WavePlayer_3, properties {soundWaveAssetPtr=&/Crow_Caw_Long_06}}
 {id=WavePlayer_4, properties {soundWaveAssetPtr=&/Crow_Caw_Long_07}}
 {id=WavePlayer_5, properties {soundWaveAssetPtr=&/Crow_Caw_Long_08}}
]
properties {
 attenuationSettings=/Irwin/Sounds/Crow/Crow_Vocalization_Att
 bOverrideConcurrency=true
 maxDistance=3300
 pitchMultiplier=0.9
 soundClassObject=&Fort_Creatures_Vocalization_Tactical
 volumeMultiplier=2.45
 concurrencyOverrides {bLimitToOwner=true, maxCount=1, resolutionRule=PreventNew}
}
