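# Sound cue that plays a short crow caw vocalization, attenuated and routed through the creature vocalization sound class.
# A quality-level switch picks a six-wave or three-wave random pool, each with an empty branch at half weight so roughly half the triggers stay silent.
# A modulator at the root varies the output, including volume.
# Overridden concurrency allows one instance per owner and blocks new plays while one is active.
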SoundCue /Irwin/Sounds/Crow/Vocalizations/Crow_Caw_Short_Cue
root=Modulator_0
nodes [
 {id=Modulator_0, children [QualityLevel_0], properties {volumeMin=0.85}}
 {id=QualityLevel_0, children [Random_2,Random_2,Random_3,Random_3]}
 {
  id=Random_2
  children [WavePlayer_2, WavePlayer_12, WavePlayer_16, WavePlayer_17, WavePlayer_18, WavePlayer_19, null]
  properties {bRandomizeWithoutReplacement=false, weights [0.083,0.083,0.083,0.083,0.083,0.083,0.5]}
 }
 {
  id=Random_3
  children [WavePlayer_20, WavePlayer_21, WavePlayer_22, null]
  properties {bRandomizeWithoutReplacement=false, bShouldExcludeFromBranchCulling=true, weights [0.16,0.16,0.16,0.5]}
 }
 {id=WavePlayer_12, properties {soundWaveAssetPtr=&/Crow_Caw_Short_02}}
 {id=WavePlayer_16, properties {soundWaveAssetPtr=&/Crow_Caw_Short_03}}
 {id=WavePlayer_17, properties {soundWaveAssetPtr=&/Crow_Caw_Short_04}}
 {id=WavePlayer_18, properties {soundWaveAssetPtr=&/Crow_Caw_Short_05}}
 {id=WavePlayer_19, properties {soundWaveAssetPtr=&/Crow_Caw_Short_06}}
 {id=WavePlayer_2, properties {soundWaveAssetPtr=&/Crow_Caw_Short_01}}
 {id=WavePlayer_20, properties {soundWaveAssetPtr=&Crow_Caw_Short_01}}
 {id=WavePlayer_21, properties {soundWaveAssetPtr=&Crow_Caw_Short_02}}
 {id=WavePlayer_22, properties {soundWaveAssetPtr=&Crow_Caw_Short_03}}
]
properties {
 attenuationSettings=/Irwin/Sounds/Crow/Crow_Vocalization_Att
 bOverrideConcurrency=true
 maxDistance=3300
 priority=3
 soundClassObject=/Game/Sounds/FortSoundClassesAndModes/Fort_Creatures_Vocalization_Normal
 volumeMultiplier=2.35
 concurrencyOverrides {bLimitToOwner=true, maxCount=1, resolutionRule=PreventNew}
 concurrencySet [null]
}
